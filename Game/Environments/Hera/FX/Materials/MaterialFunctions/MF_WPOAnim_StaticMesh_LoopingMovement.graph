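# Material function for looping VFX world position offset on spherical scatter meshes, outputting movement, opacity and sprite size.
# An input looping time plus per-element UV1 randomness drives a wrapped sweep along the direction input, scaled by object radius.
# Time-driven sines, offset by UV1, add three-axis wobble between drift limits.
# A sine of the loop phase fades opacity and scales sprite size from its minimum multiplier.
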MaterialFunction /Game/Environments/Hera/FX/Materials/MaterialFunctions/MF_WPOAnim_StaticMesh_LoopingMovement
graph {
 properties {
  description="Looping WPO movement, used for any spherical scatter meshes that you want looping movement, fading and sprite scaling on"
  userExposedCaption=VFX WPO Looping Parameters
  libraryCategoriesText [VFX]
 }
 nodes [
  Constant()
  n1 = NamedRerouteUsage()
  n2 = NamedRerouteUsage()
  n3 = Sine(None=Add(A=Add(A=@n1, B=@n2)) {constB=-0.25})
  n4 = Power(Base=LinearInterpolate(Alpha=ConstantBiasScale(None=@n3))) {constExponent=1}
  FunctionOutput(None=@n4) {description=Output to Opactity, outputName=Opacity, sortPriority=1}
  n5 = FunctionInput() {
   bUsePreviewValueAsDefault=true
   description=Sprite Size min size multiplier
   inputName=SpriteSize_MinMult
   inputType=FunctionInput_Scalar
   sortPriority=4
   previewValue=<0.5 0 0 1>
  }
  n6 = FunctionInput() {
   bUsePreviewValueAsDefault=true
   description=Sprite Size
   inputName=SpriteSize
   inputType=FunctionInput_Scalar
   sortPriority=3
   previewValue=<8 0 0 1>
  }
  FunctionOutput(None=LinearInterpolate(A=Multiply(A=@n5, B=@n6), B=@n6, Alpha=@n4)) {description=Sprite XY Scaling, outputName=Sprite Size, sortPriority=2}
  n7 = Multiply(A=Time()) {constB=0.001}
  n8 = ComponentMask(None=TextureCoordinate() {coordinateIndex=1}) {r=true}
  n9 = Add(A=Multiply(A=Time()) {constB=0.4}, B=Sine(None=Add(A=@n7, B=@n8)) {period=0.17})
  n10 = AppendVector(A=AppendVector(A=Sine(None=@n9) {period=0.56}, B=Sine(None=@n9) {period=0.7}), B=Sine(None=@n9) {period=0.82})
  n11 = FunctionInput() {
   bUsePreviewValueAsDefault=true
   desc=Sideways wobble movement
   description=Sideways wobble movement
   inputName=HorizontalDrift
   inputType=FunctionInput_Scalar
   sortPriority=2
   previewValue=<10 0 0 0>
  }
  n12 = FunctionInput() {
   bUsePreviewValueAsDefault=true
   description="Sideways wobble movement, max value"
   inputName=HorizontalDrift_Max
   inputType=FunctionInput_Scalar
   sortPriority=2
   previewValue=<100 0 0 0>
  }
  n13 = Sine(None=Add(A=@n7, B=@n8)) {period=0.14}
  n14 = LinearInterpolate(A=@n11, B=@n12, Alpha=@n13)
  n15 = FunctionInput() {
   bUsePreviewValueAsDefault=true
   description="Sideways wobble movement, max value"
   inputName=VerticalDrift
   inputType=FunctionInput_Scalar
   sortPriority=1
   previewValue=<0.6 0 0 0>
  }
  n16 = Constant() {r=0.5}
  n17 = Multiply(A=ObjectRadius(), B=LinearInterpolate(A=Multiply(A=@n15, B=@n16), B=@n16, Alpha=ConstantBiasScale(None=@n13)) {constA=0.5, constB=0.5}) {constB=0.5}
  n18 = NamedRerouteUsage()
  n19 = NamedRerouteUsage()
  n20 = LinearInterpolate(A=Multiply(A=@n17) {constB=-1}, B=@n17, Alpha=Frac(None=Add(A=@n18, B=@n19)))
  n21 = FunctionInput() {
   bUsePreviewValueAsDefault=true
   description=Overall movement direction of the mesh
   inputName=OverallDireciton
   sortPriority=1
   previewValue=<0 0 1 0>
  }
  FunctionOutput(None=Add(A=Multiply(A=@n10, B=@n14), B=Multiply(A=@n20, B=@n21))) {description=Looping WPO movement, outputName=WPO Movement}
  n22 = TextureCoordinate() {coordinateIndex=1}
  Multiply(A=Subtract(A=ComponentMask(None=@n22) {g=true}) {constB=0.5}) {constB=-2}
  n23 = FunctionInput() {
   bUsePreviewValueAsDefault=true
   description=Input looping time here to make movement looping
   inputName=InputLoopingTimeHere
   inputType=FunctionInput_Scalar
   previewValue=<1 0 0 1>
  }
  NamedRerouteDeclaration(None=@n23) {name=InputLoopingTimeHere, nodeColor=<rgba 1 0 0.5859 1>}
  n24 = TextureCoordinate() {coordinateIndex=1}
  n25 = Multiply(A=Subtract(A=ComponentMask(None=@n24) {g=true}) {constB=0.5}) {constB=-2}
  NamedRerouteDeclaration(None=@n25) {name=TexCoordRandom, nodeColor=<rgba 1 0.1875 0 1>}
  n26 = PreSkinnedLocalBounds()
  n27 = ConstantBiasScale(None=Divide(A=LocalPosition() {includedOffsets=ExcludeOffsets}.XY, B=ComponentMask(None=@n26.Half Extents) {g=true, r=true}))
  StaticSwitchParameter(True=@n25, False=ComponentMask(None=Saturate(None=@n27)) {g=true}) {defaultValue=true, parameterName=True_Texcoord1_False_Bounds}
 ]
}
